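# Modular Control Rig limb module that bends a seven-bone chain along a stretchy B-spline.
# Construction spawns proxy spline bones, five controls, and a master control with enable, visibility and volume channels.
# When enabled, controls follow three IK/FK virtual bones and their midpoints, spline IK solves the proxies, and transforms copy to the chain.
# Volume scaling and orientation-correction logic exist but are unwired, and chain scale resets to one.
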ControlRigBlueprint /EpicControlRig/Modules/Biped/LimbBend/CRM_Epic_LimbBend_v01
graphs [
 New Function_ContainedGraph {nodes [
   n1 = RigVMFunctionEntryNode()
   RigVMFunction_Sequence = &RigVMFunction_Sequence(ExecuteContext=@n1.ExecuteContext)
   n2 = &RigVMFunction_Sequence(ExecuteContext=@RigVMFunction_Sequence.A)
   Int to Name_2 = /Script/RigVM.RigVMFunction_MathIntToName(PaddedSize=0, Number=RigVMVariableNode() {variable=Index}.Value) {is=struct RigVMUnitNode, template="Int to Name::Execute(in Number,in PaddedSize,out Result)"}
   n3 = &RigUnit_HierarchyAddNull(
    Space=GlobalSpace
    Name=/Script/RigVM.RigVMFunction_NameConcat(A=Spline Null 0, B=@"Int to Name_2".Result) {is=struct RigVMUnitNode, template="Concat::Execute(in A,in B,out Result)"}.Result
    ExecutePin=@n2.A
    Transform=RigVMVariableNode() {variable=Null Transform}.Value
    Parent=RigVMVariableNode() {variable=Parent}.Value
   )
   n4 = RigVMVariableNode() {variable=Controls Scale}
   n5 = RigVMDispatchNode(Value=@n4.Value) {template="DISPATCH_RigVMDispatch_Constant(io Value)"}
   Multiply = &RigVMFunction_MathDoubleMul(B=2, A=@n4.Value)
   > n6 = /Script/ControlRig.RigUnit_HierarchyAddControlTransform(
    InitialValue=(Rotation=(X=0,Y=0,Z=0,W=1),Translation=(X=0,Y=0,Z=0),Scale3D=(X=1,Y=1,Z=1))
    OffsetSpace=GlobalSpace
    Settings="(InitialSpace=LocalSpace,bUsePreferredRotationOrder=False,PreferredRotationOrder=YZX,Limits=(LimitTranslationX=(bMinimum=False,bMaximum=False),LimitTranslationY=(bMinimum=False,bMaximum=False),LimitTranslationZ=(bMinimum=False,bMaximum=False),LimitPitch=(b ...(truncated)"
    Name=&RigVMFunction_NameConcat(A=Spline 0, B=@"Int to Name_2".Result) {template="Concat::Execute(in A,in B,out Result)"}.Result
    Settings.Shape.Transform.Scale3D.X=@n5.Value
    Settings.Shape.Transform.Scale3D.Y=@n5.Value
    Parent=@n3.Item
    Settings.Shape.Transform.Scale3D.Z=@Multiply.Result
    OffsetTransform=RigVMVariableNode() {variable=Control Transform}.Value
   ) {
    is=struct RigVMUnitNode
    template="SpawnControl::Execute(in InitialValue,in Settings,in OffsetTransform,in OffsetSpace,in Parent,in Name,out Item)"
   }
   > RigVMVariableNode(Value=@n6.Item) {variable=output}
   RigVMFunctionReturnNode(ExecuteContext=@RigVMFunction_Sequence.B, Control=RigVMVariableNode() {variable=output}.Value)
  ]}
 New Function_ContainedGraph {nodes [
   n1 = RigVMFunctionEntryNode()
   Subtract = &RigVMFunction_MathDoubleSub(B=&RigVMFunction_MathDoubleClamp(Value=@n1.volume position).Result, A=RigVMVariableNode() {variable=bone index 0 - 1}.Value)
   Absolute = &RigVMFunction_MathDoubleAbs(Value=@Subtract.Result)
   Multiply = &RigVMFunction_MathDoubleMul(A=@Absolute.Result, B=RigVMVariableNode() {variable=volume falloff}.Value)
   Clamp = &RigVMFunction_MathDoubleClamp(Minimum=-1, Value=@n1.volume amount)
   RigVMFunction_MathVectorDistance = &RigVMFunction_MathVectorDistance(
    A=&RigUnit_GetTransform(bInitial=true, Item=RigVMVariableNode() {variable=Bone Start}.Value).Transform.Translation
    B=&RigUnit_GetTransform(bInitial=true, Item=RigVMVariableNode() {variable=Bone End}.Value).Transform.Translation
   )
   n2 = RigVMVariableNode() {variable=Bone Start}
   n3 = RigVMVariableNode() {variable=Bone End}
   Divide = &RigVMFunction_MathDoubleDiv(
    A=@RigVMFunction_MathVectorDistance.Result
    B=&RigVMFunction_MathVectorDistance(
     A=&RigUnit_GetTransform(bInitial=false, Item=@n2.Value).Transform.Translation
     B=&RigUnit_GetTransform(bInitial=false, Item=@n3.Value).Transform.Translation
    ).Result
   )
   Minimum = /Script/RigVM.RigVMFunction_MathDoubleMin(B=1, A=@Divide.Result) {is=struct RigVMUnitNode, template="Minimum::Execute(in A,in B,out Result)"}
   > RigVMFunctionReturnNode(Result=/Script/RigVM.RigVMFunction_MathDoubleLerp(
     B=1
     T=&RigVMFunction_MathDoubleClamp(Value=@Multiply.Result).Result
     A=&RigVMFunction_MathDoubleLerp(A=1, T=@Clamp.Result, B=@Minimum.Result) {template="Interpolate::Execute(in A,in B,in T,out Result)"}.Result
    ) {is=struct RigVMUnitNode, template="Interpolate::Execute(in A,in B,in T,out Result)"}.Result)
  ]}
 New Function_ContainedGraph {nodes [
   RigVMFunctionEntryNode()
   > RigVMFunction_Sequence = RigVMAggregateNode()
   RigVMFunction_Sequence_1 = RigVMAggregateNode(ExecuteContext=@RigVMFunction_Sequence.B)
   SplineFromItems = RigVMFunctionReferenceNode(
    Items=((Type=Connector,Name="IKFK Virtual Bone A"),(Type=Connector,Name="IKFK Virtual Bone B"),(Type=Connector,Name="IKFK Virtual Bone C"),(Type=Connector,Name="IKFK Virtual Bone C"))
    Samples Per Segment=16
    Spline Mode=BSpline
    ExecuteContext=@RigVMFunction_Sequence_1.B
    Items.1=RigVMVariableNode() {variable=Spline Point B}.Value
    Items.2=RigVMVariableNode() {variable=Spline Point C}.Value
   )
   > n1 = RigVMVariableNode(Value=@SplineFromItems.Spline) {variable=Spline}
   RigVMFunction_ControlFlowBranch = &RigVMFunction_ControlFlowBranch(ExecuteContext=@n1.ExecuteContext, Condition=RigVMVariableNode() {variable=Draw Debug Lines}.Value)
   /Script/ControlRigSpline.RigUnit_DrawControlRigSpline(
    Color=(R=1,G=0,B=0,A=1)
    Detail=16
    Thickness=2
    ExecutePin=@RigVMFunction_ControlFlowBranch.Completed
    Spline=RigVMVariableNode() {variable=Spline}.Value
   ) {is=struct RigVMUnitNode}
   For_Each = RigVMDispatchNode(ExecuteContext=@RigVMFunction_Sequence_1.C, Array=RigVMVariableNode() {variable=Spline Bones}.Value) {template="DISPATCH_RigVMDispatch_ArrayIterator(in Array,out Element,out Index,out Count,out Ratio)"}
   > Set Transform = /Script/ControlRig.RigUnit_SetTranslation(
    Space=GlobalSpace
    Weight=1
    bInitial=False
    bPropagateToChildren=True
    Item=@For_Each.Element
    Value=/Script/ControlRigSpline.RigUnit_PositionFromControlRigSpline(U=@For_Each.Ratio, Spline=RigVMVariableNode() {variable=Spline}.Value) {is=struct RigVMUnitNode}.Position
   ) {is=struct RigVMUnitNode, template="Set Transform::Execute(in Item,in Space,in bInitial,in Value,in Weight,in bPropagateToChildren)"}
   RigVMFunctionReturnNode(ExecuteContext=@RigVMFunction_Sequence.C)
  ]}
 New Function_ContainedGraph {nodes [
   n1 = RigVMFunctionEntryNode()
   n2 = &RigVMFunction_Sequence(ExecuteContext=@n1.ExecuteContext)
   RigVMFunctionReturnNode(ExecuteContext=@n2.B)
   RigVMFunction_Sequence_2 = RigVMAggregateNode(ExecuteContext=())
   n3 = RigVMVariableNode() {variable=Spline Controls}
   Set Transform_1 = &RigUnit_SetRotation(
    Space=GlobalSpace
    Weight=1
    bInitial=False
    Item=RigVMDispatchNode(Index=-1, Array=RigVMVariableNode() {variable=Spline Bones}.Value) {template="DISPATCH_RigVMDispatch_ArrayGetAtIndex(in Array,in Index,out Element)"}.Element
    Value=&RigUnit_GetTransform(
     bInitial=False
     Item=RigVMDispatchNode(Index=-1, Array=@n3.Value) {template="DISPATCH_RigVMDispatch_ArrayGetAtIndex(in Array,in Index,out Element)"}.Element
    ).Transform.Rotation
    ExecutePin=@RigVMFunction_Sequence_2.A
   )
   n4 = RigVMVariableNode() {variable=Spline Controls}
   > Set Transform_2 = &RigUnit_SetRotation(
    Space=GlobalSpace
    Weight=1
    bInitial=False
    Item=RigVMDispatchNode(Index=3, Array=RigVMVariableNode() {variable=Spline Bones}.Value) {template="DISPATCH_RigVMDispatch_ArrayGetAtIndex(in Array,in Index,out Element)"}.Element
    Value=&RigUnit_GetTransform(
     bInitial=False
     Item=RigVMDispatchNode(Index=2, Array=@n4.Value) {template="DISPATCH_RigVMDispatch_ArrayGetAtIndex(in Array,in Index,out Element)"}.Element
    ).Transform.Rotation
   )
   n5 = RigVMDispatchNode(Index=1, Array=RigVMVariableNode() {variable=Spline Bones}.Value) {template="DISPATCH_RigVMDispatch_ArrayGetAtIndex(in Array,in Index,out Element)"}
   RigVMFunction_MathQuaternionToEuler = &RigVMFunction_MathQuaternionToEuler(RotationOrder=XYZ, Value=&RigUnit_GetTransform(bInitial=False, Item=@n5.Element).Transform.Rotation)
   Divide = &RigVMFunction_MathDoubleDiv(A=1, B=3)
   n6 = RigVMDispatchNode(Index=0, Array=RigVMVariableNode() {variable=Spline Bones}.Value) {template="DISPATCH_RigVMDispatch_ArrayGetAtIndex(in Array,in Index,out Element)"}
   n7 = RigVMDispatchNode(Index=3, Array=RigVMVariableNode() {variable=Spline Bones}.Value) {template="DISPATCH_RigVMDispatch_ArrayGetAtIndex(in Array,in Index,out Element)"}
   n8 = &RigVMFunction_MathQuaternionToEuler(
    RotationOrder=XYZ
    Value=/Script/RigVM.RigVMFunction_MathQuaternionSlerp(
     T=@Divide.Result
     A=&RigUnit_GetTransform(bInitial=False, Item=@n6.Element).Transform.Rotation
     B=&RigUnit_GetTransform(bInitial=False, Item=@n7.Element).Transform.Rotation
    ) {is=struct RigVMUnitNode, template="Interpolate::Execute(in A,in B,in T,out Result)"}.Result
   )
   RigVMFunction_MathQuaternionFromEuler = &RigVMFunction_MathQuaternionFromEuler(Euler=(X=0,Y=0,Z=0), RotationOrder=XYZ, Euler.X=@RigVMFunction_MathQuaternionToEuler.Result.X, Euler.Y=@n8.Result.Y, Euler.Z=@n8.Result.Z)
   Set Transform_3 = &RigUnit_SetRotation(ExecutePin=(), Space=GlobalSpace, Weight=1, bInitial=False, Item=@n5.Element, Value=@RigVMFunction_MathQuaternionFromEuler.Result)
   n9 = RigVMDispatchNode(Index=2, Array=RigVMVariableNode() {variable=Spline Bones}.Value) {template="DISPATCH_RigVMDispatch_ArrayGetAtIndex(in Array,in Index,out Element)"}
   n10 = &RigUnit_GetTransform(bInitial=False, Item=@n9.Element)
   n11 = RigVMDispatchNode(Index=0, Array=RigVMVariableNode() {variable=Spline Bones}.Value) {template="DISPATCH_RigVMDispatch_ArrayGetAtIndex(in Array,in Index,out Element)"}
   n12 = RigVMDispatchNode(Index=3, Array=RigVMVariableNode() {variable=Spline Bones}.Value) {template="DISPATCH_RigVMDispatch_ArrayGetAtIndex(in Array,in Index,out Element)"}
   n13 = &RigVMFunction_MathQuaternionToEuler(
    RotationOrder=XYZ
    Value=&RigVMFunction_MathQuaternionSlerp(
     T=&RigVMFunction_MathDoubleDiv(A=2, B=3).Result
     A=&RigUnit_GetTransform(bInitial=False, Item=@n11.Element).Transform.Rotation
     B=&RigUnit_GetTransform(bInitial=False, Item=@n12.Element).Transform.Rotation
    ) {template="Interpolate::Execute(in A,in B,in T,out Result)"}.Result
   )
   Set Transform_4 = &RigUnit_SetRotation(
    ExecutePin=()
    Space=GlobalSpace
    Weight=1
    bInitial=False
    Item=@n9.Element
    Value=&RigVMFunction_MathQuaternionFromEuler(
     Euler=(X=0,Y=0,Z=0)
     RotationOrder=XYZ
     Euler.X=&RigVMFunction_MathQuaternionToEuler(RotationOrder=XYZ, Value=@n10.Transform.Rotation).Result.X
     Euler.Y=@n13.Result.Y
     Euler.Z=@n13.Result.Z
    ).Result
   )
   n14 = RigVMDispatchNode(Index=4, Array=RigVMVariableNode() {variable=Spline Bones}.Value) {template="DISPATCH_RigVMDispatch_ArrayGetAtIndex(in Array,in Index,out Element)"}
   n15 = RigVMDispatchNode(Index=3, Array=RigVMVariableNode() {variable=Spline Bones}.Value) {template="DISPATCH_RigVMDispatch_ArrayGetAtIndex(in Array,in Index,out Element)"}
   n16 = RigVMDispatchNode(Index=6, Array=RigVMVariableNode() {variable=Spline Bones}.Value) {template="DISPATCH_RigVMDispatch_ArrayGetAtIndex(in Array,in Index,out Element)"}
   n17 = &RigVMFunction_MathQuaternionToEuler(
    RotationOrder=XYZ
    Value=&RigVMFunction_MathQuaternionSlerp(
     T=&RigVMFunction_MathDoubleDiv(A=1, B=3).Result
     A=&RigUnit_GetTransform(bInitial=False, Item=@n15.Element).Transform.Rotation
     B=&RigUnit_GetTransform(bInitial=False, Item=@n16.Element).Transform.Rotation
    ) {template="Interpolate::Execute(in A,in B,in T,out Result)"}.Result
   )
   n18 = &RigUnit_GetTransform(bInitial=False, Item=@n14.Element)
   Set Transform_5 = &RigUnit_SetRotation(
    ExecutePin=()
    Space=GlobalSpace
    Weight=1
    bInitial=False
    Item=@n14.Element
    Value=&RigVMFunction_MathQuaternionFromEuler(
     Euler=(X=0,Y=0,Z=0)
     RotationOrder=XYZ
     Euler.Y=@n17.Result.Y
     Euler.Z=@n17.Result.Z
     Euler.X=&RigVMFunction_MathQuaternionToEuler(RotationOrder=XYZ, Value=@n18.Transform.Rotation).Result.X
    ).Result
   )
   n19 = RigVMDispatchNode(Index=5, Array=RigVMVariableNode() {variable=Spline Bones}.Value) {template="DISPATCH_RigVMDispatch_ArrayGetAtIndex(in Array,in Index,out Element)"}
   n20 = &RigUnit_GetTransform(bInitial=False, Item=@n19.Element)
   n21 = RigVMDispatchNode(Index=6, Array=RigVMVariableNode() {variable=Spline Bones}.Value) {template="DISPATCH_RigVMDispatch_ArrayGetAtIndex(in Array,in Index,out Element)"}
   n22 = RigVMDispatchNode(Index=3, Array=RigVMVariableNode() {variable=Spline Bones}.Value) {template="DISPATCH_RigVMDispatch_ArrayGetAtIndex(in Array,in Index,out Element)"}
   n23 = &RigVMFunction_MathQuaternionToEuler(
    RotationOrder=XYZ
    Value=&RigVMFunction_MathQuaternionSlerp(
     T=&RigVMFunction_MathDoubleDiv(A=2, B=3).Result
     B=&RigUnit_GetTransform(bInitial=False, Item=@n21.Element).Transform.Rotation
     A=&RigUnit_GetTransform(bInitial=False, Item=@n22.Element).Transform.Rotation
    ) {template="Interpolate::Execute(in A,in B,in T,out Result)"}.Result
   )
   Set Transform_6 = &RigUnit_SetRotation(
    ExecutePin=()
    Space=GlobalSpace
    Weight=1
    bInitial=False
    Item=@n19.Element
    Value=&RigVMFunction_MathQuaternionFromEuler(
     Euler=(X=0,Y=0,Z=0)
     RotationOrder=XYZ
     Euler.X=&RigVMFunction_MathQuaternionToEuler(RotationOrder=XYZ, Value=@n20.Transform.Rotation).Result.X
     Euler.Y=@n23.Result.Y
     Euler.Z=@n23.Result.Z
    ).Result
   )
  ]}
 RigVMFunctionLibrary {nodes [
   GetVolumeScale = RigVMCollapseNode(
    Bone End=(Type=None,Name="None")
    Bone Start=(Type=None,Name="None")
    bone index 0 - 1=0
    volume amount=0
    volume falloff=1.5
    volume position=0
   )
   Limb Spline = RigVMCollapseNode(ExecuteContext=())
   Spawn Control at Position = RigVMCollapseNode(
    Control Transform=(Rotation=(X=0,Y=0,Z=0,W=1),Translation=(X=0,Y=0,Z=0),Scale3D=(X=1,Y=1,Z=1))
    ExecuteContext=()
    Index=0
    Null Transform=(Rotation=(X=0,Y=0,Z=0,W=1),Translation=(X=0,Y=0,Z=0),Scale3D=(X=1,Y=1,Z=1))
   )
   Spline Orient Correction = RigVMCollapseNode(ExecuteContext=())
  ]}
 RigVMFunction_Sequence_1_ContainedGraph {nodes [
   n1 = RigVMFunctionEntryNode()
   RigVMFunction_Sequence = &RigVMFunction_Sequence(ExecuteContext=@n1.ExecuteContext)
   n2 = &RigVMFunction_Sequence(ExecuteContext=@RigVMFunction_Sequence.B)
   n3 = &RigVMFunction_Sequence(ExecuteContext=@n2.B)
   n4 = &RigVMFunction_Sequence(ExecuteContext=@n3.B)
   n5 = &RigVMFunction_Sequence(ExecuteContext=@n4.B)
   n6 = &RigVMFunction_Sequence(ExecuteContext=@n5.B)
   RigVMFunctionReturnNode(A=@RigVMFunction_Sequence.A, B=@n2.A, C=@n3.A, D=@n4.A, E=@n5.A, F=@n6.A, G=@n6.B)
  ]}
 RigVMFunction_Sequence_1_ContainedGraph {nodes [
   n1 = RigVMFunctionEntryNode()
   RigVMFunction_Sequence = &RigVMFunction_Sequence(ExecuteContext=@n1.ExecuteContext)
   n2 = &RigVMFunction_Sequence(ExecuteContext=@RigVMFunction_Sequence.B)
   n3 = &RigVMFunction_Sequence(ExecuteContext=@n2.B)
   n4 = &RigVMFunction_Sequence(ExecuteContext=@n3.B)
   n5 = &RigVMFunction_Sequence(ExecuteContext=@n4.B)
   n6 = &RigVMFunction_Sequence(ExecuteContext=@n5.B)
   RigVMFunctionReturnNode(A=@RigVMFunction_Sequence.A, B=@n2.A, C=@n3.A, D=@n4.A, E=@n5.A, F=@n6.A, G=@n6.B)
  ]}
 RigVMFunction_Sequence_1_ContainedGraph {nodes [
   n1 = RigVMFunctionEntryNode()
   RigVMFunction_Sequence = &RigVMFunction_Sequence(ExecuteContext=@n1.ExecuteContext)
   n2 = &RigVMFunction_Sequence(ExecuteContext=@RigVMFunction_Sequence.B)
   n3 = &RigVMFunction_Sequence(ExecuteContext=@n2.B)
   n4 = &RigVMFunction_Sequence(ExecuteContext=@n3.B)
   n5 = &RigVMFunction_Sequence(ExecuteContext=@n4.B)
   n6 = &RigVMFunction_Sequence(ExecuteContext=@n5.B)
   RigVMFunctionReturnNode(A=@RigVMFunction_Sequence.A, B=@n2.A, C=@n3.A, D=@n4.A, E=@n5.A, F=@n6.A, G=@n6.B)
  ]}
 RigVMFunction_Sequence_2_ContainedGraph {nodes [
   n1 = RigVMFunctionEntryNode()
   RigVMFunction_Sequence = &RigVMFunction_Sequence(ExecuteContext=@n1.ExecuteContext)
   n2 = &RigVMFunction_Sequence(ExecuteContext=@RigVMFunction_Sequence.B)
   RigVMFunctionReturnNode(A=@RigVMFunction_Sequence.A, B=@n2.A, C=@n2.B)
  ]}
 RigVMFunction_Sequence_2_ContainedGraph {nodes [
   n1 = RigVMFunctionEntryNode()
   n2 = &RigVMFunction_Sequence(ExecuteContext=@n1.ExecuteContext)
   n3 = &RigVMFunction_Sequence(ExecuteContext=@n2.B)
   n4 = &RigVMFunction_Sequence(ExecuteContext=@n3.B)
   n5 = &RigVMFunction_Sequence(ExecuteContext=@n4.B)
   RigVMFunctionReturnNode(A=@n2.A, B=@n3.A, C=@n4.A, D=@n5.A, E=@n5.B)
  ]}
 RigVMFunction_Sequence_3_ContainedGraph {nodes [
   n1 = RigVMFunctionEntryNode()
   n2 = &RigVMFunction_Sequence(ExecuteContext=@n1.ExecuteContext)
   n3 = &RigVMFunction_Sequence(ExecuteContext=@n2.B)
   RigVMFunctionReturnNode(B=@n3.A, C=@n3.B, A=@n2.A)
  ]}
 RigVMModel {nodes [
   RigUnit_BeginExecution = &RigUnit_BeginExecution()
   n1 = &RigVMFunction_ControlFlowBranch(
    Condition=/Script/ControlRig.RigUnit_GetBoolAnimationChannelFromItem(bInitial=False, Item=RigVMVariableNode() {variable=Spline Switch Channel}.Value) {is=struct RigVMUnitNode, template="GetAnimationChannelFromItem::Execute(out Value,in Item,in bInitial)"}.Value
    ExecuteContext=@RigUnit_BeginExecution.ExecutePin
   )
   RigVMFunction_Sequence_1 = RigVMAggregateNode(ExecuteContext=@n1.Completed)
   For_Each_1 = RigVMDispatchNode(ExecuteContext=@RigVMFunction_Sequence_1.B, Array=RigVMVariableNode() {variable=Spline Controls}.Value) {template="DISPATCH_RigVMDispatch_ArrayIterator(in Array,out Element,out Index,out Count,out Ratio)"}
   > SetControlVisibility = /Script/ControlRig.RigUnit_SetControlVisibility(
    Item=@For_Each_1.Element
    bVisible=&RigUnit_GetBoolAnimationChannelFromItem(bInitial=False, Item=RigVMVariableNode() {variable=Sub Control Visibility Channel}.Value) {template="GetAnimationChannelFromItem::Execute(out Value,in Item,in bInitial)"}.Value
   ) {is=struct RigVMUnitNode, template="SetControlVisibility::Execute(in Item,in Pattern,in bVisible)"}
   n2 = RigVMVariableNode() {variable=Draw Debug Lines}
   RigVMFunction_ControlFlowBranch = &RigVMFunction_ControlFlowBranch(Condition=@n2.Value, ExecuteContext=@RigVMFunction_Sequence_1.A)
   n3 = RigVMVariableNode() {variable=Spline Bones}
   RigVMFunction_DebugTransformArrayMutableNoSpace = /Script/RigVM.RigVMFunction_DebugTransformArrayMutableNoSpace(
    Color=(R=1,G=1,B=1,A=1)
    DebugDrawSettings=(DepthPriority=SDPG_Foreground,Lifetime=-1)
    Mode=Axes
    ParentIndices=()
    Scale=10
    Thickness=0.3
    WorldOffset=(Rotation=(X=0,Y=0,Z=0,W=1),Translation=(X=0,Y=0,Z=0),Scale3D=(X=1,Y=1,Z=1))
    Transforms=/Script/ControlRig.RigUnit_GetTransformItemArray(Space=GlobalSpace, bInitial=False, Items=@n3.Value) {is=struct RigVMUnitNode}.Transforms
    ExecutePin=@RigVMFunction_ControlFlowBranch.Completed
    bEnabled=@n2.Value
   ) {is=struct RigVMUnitNode}
   SplineFromItems = RigVMFunctionReferenceNode(Samples Per Segment=16, Spline Mode=Hermite, ExecuteContext=@RigVMFunction_ControlFlowBranch.True, Items=@n3.Value)
   > &RigUnit_DrawControlRigSpline(Color=(R=0,G=1,B=0,A=1), Detail=16, Thickness=2, Spline=@SplineFromItems.Spline)
   RigVMFunction_Sequence = RigVMAggregateNode(ExecuteContext=@n1.True)
   n4 = RigVMVariableNode() {variable=Spline Controls}
   RigUnit_GetTransform = &RigUnit_GetTransform(
    bInitial=False
    Item=RigVMDispatchNode(Index=0, Array=@n4.Value) {template="DISPATCH_RigVMDispatch_ArrayGetAtIndex(in Array,in Index,out Element)"}.Element
   )
   Interpolate = /Script/RigVM.RigVMFunction_MathTransformLerp(
    T=0.5
    A=@RigUnit_GetTransform.Transform
    B=&RigUnit_GetTransform(
     bInitial=False
     Item=RigVMDispatchNode(Index=2, Array=@n4.Value) {template="DISPATCH_RigVMDispatch_ArrayGetAtIndex(in Array,in Index,out Element)"}.Element
    ).Transform
   ) {is=struct RigVMUnitNode, template="Interpolate::Execute(in A,in B,in T,out Result)"}
   n5 = /Script/ControlRig.RigUnit_HierarchyGetParent(
    bDefaultParent=True
    Child=RigVMDispatchNode(Index=1, Array=@n4.Value) {template="DISPATCH_RigVMDispatch_ArrayGetAtIndex(in Array,in Index,out Element)"}.Element
   ) {is=struct RigVMUnitNode}
   RigUnit_SetRotation = &RigUnit_SetRotation(
    Space=GlobalSpace
    Weight=1
    bInitial=False
    bPropagateToChildren=true
    Value=@Interpolate.Result.Rotation
    ExecutePin=@RigVMFunction_Sequence.C
    Item=@n5.Parent
   )
   > RigUnit_SetTranslation = &RigUnit_SetTranslation(Space=GlobalSpace, Weight=1, bInitial=False, bPropagateToChildren=True, Value=@Interpolate.Result.Translation, Item=@n5.Parent) {template="Set Transform::Execute(in Item,in Space,in bInitial,in Value,in Weight,in bPropagateToChildren)"}
   n6 = RigVMVariableNode() {variable=Spline Controls}
   n7 = &RigVMFunction_MathTransformLerp(
    T=0.5
    A=&RigUnit_GetTransform(
     bInitial=False
     Item=RigVMDispatchNode(Index=2, Array=@n6.Value) {template="DISPATCH_RigVMDispatch_ArrayGetAtIndex(in Array,in Index,out Element)"}.Element
    ).Transform
    B=&RigUnit_GetTransform(
     bInitial=False
     Item=RigVMDispatchNode(Index=4, Array=@n6.Value) {template="DISPATCH_RigVMDispatch_ArrayGetAtIndex(in Array,in Index,out Element)"}.Element
    ).Transform
   ) {template="Interpolate::Execute(in A,in B,in T,out Result)"}
   n8 = &RigUnit_HierarchyGetParent(
    bDefaultParent=True
    Child=RigVMDispatchNode(Index=3, Array=@n6.Value) {template="DISPATCH_RigVMDispatch_ArrayGetAtIndex(in Array,in Index,out Element)"}.Element
   )
   > &RigUnit_SetRotation(Space=GlobalSpace, Weight=1, bInitial=False, Value=@n7.Result.Rotation, Item=@n8.Parent)
   > &RigUnit_SetTranslation(Space=GlobalSpace, Weight=1, bInitial=False, bPropagateToChildren=true, Value=@n7.Result.Translation, Item=@n8.Parent) {template="Set Transform::Execute(in Item,in Space,in bInitial,in Value,in Weight,in bPropagateToChildren)"}
   n9 = RigVMVariableNode() {variable=Spline Controls}
   &RigUnit_SetTransform(
    Weight=1
    bInitial=False
    bPropagateToChildren=true
    Item=&RigUnit_HierarchyGetParent(
     bDefaultParent=True
     Child=RigVMDispatchNode(Index=0, Array=@n9.Value) {template="DISPATCH_RigVMDispatch_ArrayGetAtIndex(in Array,in Index,out Element)"}.Element
    ).Parent
    Value=&RigUnit_GetTransform(Item=(Type=Connector,Name="IKFK Virtual Bone A"), bInitial=False).Transform
    ExecutePin=@RigVMFunction_Sequence.B
   )
   n10 = RigVMVariableNode() {variable=Spline Controls}
   > &RigUnit_SetTransform(
    Weight=1
    bInitial=False
    bPropagateToChildren=true
    Item=&RigUnit_HierarchyGetParent(
     bDefaultParent=True
     Child=RigVMDispatchNode(Index=2, Array=@n10.Value) {template="DISPATCH_RigVMDispatch_ArrayGetAtIndex(in Array,in Index,out Element)"}.Element
    ).Parent
    Value=&RigUnit_GetTransform(Item=(Type=Connector,Name="IKFK Virtual Bone B"), bInitial=False).Transform
   )
   n11 = RigVMVariableNode() {variable=Spline Controls}
   > &RigUnit_SetTransform(
    Weight=1
    bInitial=False
    bPropagateToChildren=true
    Item=&RigUnit_HierarchyGetParent(
     bDefaultParent=True
     Child=RigVMDispatchNode(Index=4, Array=@n11.Value) {template="DISPATCH_RigVMDispatch_ArrayGetAtIndex(in Array,in Index,out Element)"}.Element
    ).Parent
    Value=&RigUnit_GetTransform(Item=(Type=Connector,Name="IKFK Virtual Bone C"), bInitial=False).Transform
   )
   SplineIK_2 = RigVMFunctionReferenceNode(
    Primary Axis=(X=1,Y=0,Z=0)
    Samples Per Segment=16
    Secondary Spline Direction=(X=0,Y=5,Z=0)
    Spline Mode=BSpline
    Stretch=true
    Up Axis=(X=0,Y=1,Z=0)
    debug=false
    Primary Axis.X=RigVMDispatchNode(False=1, True=-1, Condition=RigVMVariableNode() {variable=Mirror Behaviour}.Value) {template="DISPATCH_RigVMDispatch_If(in Condition,in True,in False,out Result)"}.Result
    Bones=RigVMVariableNode() {variable=Spline Bones}.Value
    Controls=RigVMVariableNode() {variable=Spline Controls}.Value
    ExecuteContext=@RigVMFunction_Sequence.D
   )
   For_Each = RigVMDispatchNode(
    Array=/Script/ControlRig.RigUnit_ResolveArrayConnector(Connector=(Type=Connector,Name="Bones Chain"), SkipSocket=False) {is=struct RigVMUnitNode}.Result
    ExecuteContext=@RigVMFunction_Sequence.E
   ) {template="DISPATCH_RigVMDispatch_ArrayIterator(in Array,out Element,out Index,out Count,out Ratio)"}
   n12 = RigVMVariableNode() {variable=Spline Bones}
   > &RigUnit_SetTransform(
    Weight=1
    bInitial=False
    Item=@For_Each.Element
    Value=&RigUnit_GetTransform(
     bInitial=False
     Item=RigVMDispatchNode(Index=@For_Each.Index, Array=@n12.Value) {template="DISPATCH_RigVMDispatch_ArrayGetAtIndex(in Array,in Index,out Element)"}.Element
    ).Transform
   )
   Spline Orient Correction = RigVMFunctionReferenceNode(ExecuteContext=@For_Each.Completed)
   For_Each_4 = RigVMDispatchNode(
    Array=&RigUnit_ResolveArrayConnector(Connector=(Type=Connector,Name="Bones Chain"), SkipSocket=False).Result
    ExecuteContext=@RigVMFunction_Sequence.F
   ) {template="DISPATCH_RigVMDispatch_ArrayIterator(in Array,out Element,out Index,out Count,out Ratio)"}
   > /Script/ControlRig.RigUnit_SetScale(Scale=(X=1,Y=1,Z=1), Space=GlobalSpace, Weight=1, bInitial=False, bPropagateToChildren=false, Item=@For_Each_4.Element) {is=struct RigVMUnitNode}
   RigVMDispatchNode(Index=6, Array=RigVMVariableNode() {variable=Spline Bones}.Value) {template="DISPATCH_RigVMDispatch_ArrayGetAtIndex(in Array,in Index,out Element)"}
   RigVMDispatchNode(Index=0, Array=RigVMVariableNode() {variable=Spline Bones}.Value) {template="DISPATCH_RigVMDispatch_ArrayGetAtIndex(in Array,in Index,out Element)"}
   RigVMDispatchNode(Index=3, Array=RigVMVariableNode() {variable=Spline Bones}.Value) {template="DISPATCH_RigVMDispatch_ArrayGetAtIndex(in Array,in Index,out Element)"}
   GetAnimationChannelFromItem = /Script/ControlRig.RigUnit_GetFloatAnimationChannelFromItem(bInitial=False, Item=RigVMVariableNode() {variable=Volume Amount Channel}.Value) {is=struct RigVMUnitNode, template="GetAnimationChannelFromItem::Execute(out Value,in Item,in bInitial)"}
   GetVolumeScale_4 = RigVMFunctionReferenceNode(
    Bone End=(Type=Connector,Name="IKFK Virtual Bone C")
    Bone Start=(Type=Connector,Name="IKFK Virtual Bone A")
    bone index 0 - 1=@For_Each_4.Ratio
    volume amount=@GetAnimationChannelFromItem.Value
    volume falloff=&RigUnit_GetFloatAnimationChannelFromItem(bInitial=False, Item=RigVMVariableNode() {variable=Volume Falloff Channel}.Value) {template="GetAnimationChannelFromItem::Execute(out Value,in Item,in bInitial)"}.Value
    volume position=&RigUnit_GetFloatAnimationChannelFromItem(bInitial=False, Item=RigVMVariableNode() {variable=Volume Position Channel}.Value) {template="GetAnimationChannelFromItem::Execute(out Value,in Item,in bInitial)"}.Value
   )
   Limb Spline = RigVMFunctionReferenceNode(ExecuteContext=())
   RigVMVariableNode() {variable=Draw Debug Lines}
  ]}
 RigVMModel Construct {nodes [
   n1 = &RigUnit_PrepareForExecution()
   RigVMFunction_Sequence = RigVMAggregateNode(ExecuteContext=@n1.ExecutePin)
   RigVMDispatchNode(ExecuteContext=@RigVMFunction_Sequence.A, Array=RigVMVariableNode() {variable=Spline Bones}.Value) {template="DISPATCH_RigVMDispatch_ArrayReset(io Array)"}
   > RigVMDispatchNode(Array=RigVMVariableNode() {variable=Spline Controls}.Value) {template="DISPATCH_RigVMDispatch_ArrayReset(io Array)"}
   > RigVMVariableNode(Value=/Script/ControlRig.RigUnit_ResolveConnector(Connector=(Type=Connector,Name="Parent"), SkipSocket=False) {is=struct RigVMUnitNode}.Result) {variable=Parent}
   n2 = &RigUnit_ResolveArrayConnector(Connector=(Type=Connector,Name="Bones Chain"), SkipSocket=False)
   Make_Array = RigVMDispatchNode(
    Values=((Type=None,Name="None"),((),Type=None,Name="None"),((),Type=None,Name="None"))
    Values.0=RigVMDispatchNode(Index=0, Array=@n2.Result) {template="DISPATCH_RigVMDispatch_ArrayGetAtIndex(in Array,in Index,out Element)"}.Element
    Values.1=RigVMDispatchNode(Index=3, Array=@n2.Result) {template="DISPATCH_RigVMDispatch_ArrayGetAtIndex(in Array,in Index,out Element)"}.Element
    Values.2=RigVMDispatchNode(Index=6, Array=@n2.Result) {template="DISPATCH_RigVMDispatch_ArrayGetAtIndex(in Array,in Index,out Element)"}.Element
   ) {template="DISPATCH_RigVMDispatch_ArrayMake(in Values,out Array)"}
   For_Each = RigVMDispatchNode(Array=@Make_Array.Array, ExecuteContext=@RigVMFunction_Sequence.C) {template="DISPATCH_RigVMDispatch_ArrayIterator(in Array,out Element,out Index,out Count,out Ratio)"}
   Make_Array_1 = RigVMDispatchNode(
    Values=((Type=None,Name="None"),((),Type=None,Name="None"),((),Type=None,Name="None"))
    Values.0=&RigUnit_ResolveConnector(Connector=(Type=Connector,Name="IKFK Virtual Bone A"), SkipSocket=False).Result
    Values.1=&RigUnit_ResolveConnector(Connector=(Type=Connector,Name="IKFK Virtual Bone B"), SkipSocket=False).Result
    Values.2=&RigUnit_ResolveConnector(Connector=(Type=Connector,Name="IKFK Virtual Bone C"), SkipSocket=False).Result
   ) {template="DISPATCH_RigVMDispatch_ArrayMake(in Values,out Array)"}
   Multiply = /Script/RigVM.RigVMFunction_MathIntMul(B=2, A=@For_Each.Index) {is=struct RigVMUnitNode, template="Multiply::Execute(in A,in B,out Result)"}
   > Spawn Control at Position_1 = RigVMFunctionReferenceNode(
    Control Transform=&RigUnit_GetTransform(bInitial=False, Item=@For_Each.Element).Transform
    Null Transform=&RigUnit_GetTransform(
     bInitial=False
     Item=RigVMDispatchNode(Index=@For_Each.Index, Array=@Make_Array_1.Array) {template="DISPATCH_RigVMDispatch_ArrayGetAtIndex(in Array,in Index,out Element)"}.Element
    ).Transform
    Index=@Multiply.Result
   )
   n3 = RigVMVariableNode() {variable=Spline Controls}
   > RigVMDispatchNode(Element=@"Spawn Control at Position_1".Control, Array=@n3.Value) {template="DISPATCH_RigVMDispatch_ArrayAdd(io Array,in Element,out Index)"}
   > RigVMDispatchNode(Element=(Type=None,Name="None"), Array=@n3.Value) {template="DISPATCH_RigVMDispatch_ArrayAdd(io Array,in Element,out Index)"}
   RigVMDispatchNode(Index=-1, ExecuteContext=@For_Each.Completed, Array=RigVMVariableNode() {variable=Spline Controls}.Value) {template="DISPATCH_RigVMDispatch_ArrayRemove(io Array,in Index)"}
   n4 = RigVMDispatchNode(Value=1) {template="DISPATCH_RigVMDispatch_Constant(io Value)"}
   n5 = &RigUnit_ResolveArrayConnector(Connector=(Type=Connector,Name="Bones Chain"), SkipSocket=False)
   Interpolate = &RigVMFunction_MathTransformLerp(
    T=0.5
    A=&RigUnit_GetTransform(
     bInitial=False
     Item=RigVMDispatchNode(Index=1, Array=@n5.Result) {template="DISPATCH_RigVMDispatch_ArrayGetAtIndex(in Array,in Index,out Element)"}.Element
    ).Transform
    B=&RigUnit_GetTransform(
     bInitial=False
     Item=RigVMDispatchNode(Index=2, Array=@n5.Result) {template="DISPATCH_RigVMDispatch_ArrayGetAtIndex(in Array,in Index,out Element)"}.Element
    ).Transform
   ) {template="Interpolate::Execute(in A,in B,in T,out Result)"}
   Spawn Control at Position = RigVMFunctionReferenceNode(Index=@n4.Value, Control Transform=@Interpolate.Result, Null Transform=@Interpolate.Result, ExecuteContext=@RigVMFunction_Sequence.D)
   > Set_At = RigVMDispatchNode(Index=@n4.Value, Element=@"Spawn Control at Position".Control, Array=RigVMVariableNode() {variable=Spline Controls}.Value) {template="DISPATCH_RigVMDispatch_ArraySetAtIndex(io Array,in Index,in Element)"}
   n6 = RigVMDispatchNode(Value=3) {template="DISPATCH_RigVMDispatch_Constant(io Value)"}
   n7 = &RigUnit_ResolveArrayConnector(Connector=(Type=Connector,Name="Bones Chain"), SkipSocket=False)
   n8 = &RigVMFunction_MathTransformLerp(
    T=0.5
    A=&RigUnit_GetTransform(
     bInitial=False
     Item=RigVMDispatchNode(Index=4, Array=@n7.Result) {template="DISPATCH_RigVMDispatch_ArrayGetAtIndex(in Array,in Index,out Element)"}.Element
    ).Transform
    B=&RigUnit_GetTransform(
     bInitial=False
     Item=RigVMDispatchNode(Index=5, Array=@n7.Result) {template="DISPATCH_RigVMDispatch_ArrayGetAtIndex(in Array,in Index,out Element)"}.Element
    ).Transform
   ) {template="Interpolate::Execute(in A,in B,in T,out Result)"}
   Spawn Control at Position_2 = RigVMFunctionReferenceNode(Index=@n6.Value, Control Transform=@n8.Result, Null Transform=@n8.Result, ExecuteContext=@RigVMFunction_Sequence.E)
   > Set_At_1 = RigVMDispatchNode(Index=@n6.Value, Element=@"Spawn Control at Position_2".Control, Array=RigVMVariableNode() {variable=Spline Controls}.Value) {template="DISPATCH_RigVMDispatch_ArraySetAtIndex(io Array,in Index,in Element)"}
   /Script/ControlRig.RigUnit_HierarchyAddAnimationChannelBool(
    InitialValue=false
    LimitsEnabled=()
    MaximumValue=True
    MinimumValue=False
    Name=Weapon Control Visible
    ExecutePin=@RigVMFunction_Sequence.G
    Parent=RigVMVariableNode() {variable=Spline Master Control}.Value
   ) {
    is=struct RigVMUnitNode
    template="SpawnAnimationChannel::Execute(in InitialValue,in MinimumValue,in MaximumValue,in LimitsEnabled,in Parent,in Name,out Item)"
   }
   > &RigUnit_HierarchyAddAnimationChannelBool(
    InitialValue=false
    LimitsEnabled=()
    MaximumValue=True
    MinimumValue=False
    Name=Righthand Crownball Control Visible
    Parent=RigVMVariableNode() {variable=Spline Master Control}.Value
   ) {template="SpawnAnimationChannel::Execute(in InitialValue,in MinimumValue,in MaximumValue,in LimitsEnabled,in Parent,in Name,out Item)"}
   SpawnControl = &RigUnit_HierarchyAddControlTransform(
    InitialValue=(Rotation=(X=0,Y=0,Z=0,W=1),Translation=(X=0,Y=0,Z=0),Scale3D=(X=1,Y=1,Z=1))
    Name=Spline Control
    OffsetSpace=LocalSpace
    Parent=&RigUnit_ResolveConnector(Connector=(Type=Connector,Name="IKFK Virtual Bone C"), SkipSocket=False).Result
    Settings=RigVMVariableNode() {variable=Spline Control Settings}.Value
    OffsetTransform=RigVMVariableNode() {variable=Spline Control Offset}.Value
    ExecutePin=@RigVMFunction_Sequence.F
   ) {template="SpawnControl::Execute(in InitialValue,in Settings,in OffsetTransform,in OffsetSpace,in Parent,in Name,out Item)"}
   > RigVMVariableNode(Value=@SpawnControl.Item) {variable=Spline Master Control}
   > n9 = &RigUnit_HierarchyAddAnimationChannelBool(
    InitialValue=False
    LimitsEnabled=()
    MaximumValue=True
    MinimumValue=False
    Name=Sub Controls Visible
    Parent=RigVMVariableNode() {variable=Spline Master Control}.Value
   ) {template="SpawnAnimationChannel::Execute(in InitialValue,in MinimumValue,in MaximumValue,in LimitsEnabled,in Parent,in Name,out Item)"}
   > RigVMVariableNode(Value=@n9.Item) {variable=Sub Control Visibility Channel}
   > SpawnAnimationChannel = /Script/ControlRig.RigUnit_HierarchyAddAnimationChannelFloat(
    InitialValue=0.5
    LimitsEnabled=(Enabled=(bMinimum=True,bMaximum=True))
    MaximumValue=5
    MinimumValue=-2
    Name=Volume Amount
    Parent=RigVMVariableNode() {variable=Spline Master Control}.Value
   ) {
    is=struct RigVMUnitNode
    template="SpawnAnimationChannel::Execute(in InitialValue,in MinimumValue,in MaximumValue,in LimitsEnabled,in Parent,in Name,out Item)"
   }
   > RigVMVariableNode(Value=@SpawnAnimationChannel.Item) {variable=Volume Amount Channel}
   > n10 = &RigUnit_HierarchyAddAnimationChannelFloat(
    InitialValue=1
    LimitsEnabled=(Enabled=(bMinimum=True,bMaximum=True))
    MaximumValue=5
    MinimumValue=-2
    Name=Volume Falloff
    Parent=RigVMVariableNode() {variable=Spline Master Control}.Value
   ) {template="SpawnAnimationChannel::Execute(in InitialValue,in MinimumValue,in MaximumValue,in LimitsEnabled,in Parent,in Name,out Item)"}
   > RigVMVariableNode(Value=@n10.Item) {variable=Volume Falloff Channel}
   > n11 = &RigUnit_HierarchyAddAnimationChannelFloat(
    InitialValue=0.5
    LimitsEnabled=(Enabled=(bMinimum=True,bMaximum=True))
    MaximumValue=1
    MinimumValue=0
    Name=Volume Position
    Parent=RigVMVariableNode() {variable=Spline Master Control}.Value
   ) {template="SpawnAnimationChannel::Execute(in InitialValue,in MinimumValue,in MaximumValue,in LimitsEnabled,in Parent,in Name,out Item)"}
   > RigVMVariableNode(Value=@n11.Item) {variable=Volume Position Channel}
   > n12 = &RigUnit_HierarchyAddAnimationChannelBool(
    LimitsEnabled=()
    MaximumValue=True
    MinimumValue=False
    Name=Enable Spline
    Parent=RigVMVariableNode() {variable=Spline Master Control}.Value
    InitialValue=RigVMVariableNode() {variable=Enable}.Value
   ) {template="SpawnAnimationChannel::Execute(in InitialValue,in MinimumValue,in MaximumValue,in LimitsEnabled,in Parent,in Name,out Item)"}
   > RigVMVariableNode(Value=@n12.Item) {variable=Spline Switch Channel}
   n13 = &RigUnit_ResolveArrayConnector(Connector=(Type=Connector,Name="Bones Chain"), SkipSocket=False)
   n14 = /Script/ControlRig.RigUnit_HierarchyAddBone(
    Name=Spline Point B
    Parent=(Type=Connector,Name="IKFK Virtual Bone A")
    Space=GlobalSpace
    Transform=&RigUnit_GetTransform(
     bInitial=False
     Item=RigVMDispatchNode(Index=2, Array=@n13.Result) {template="DISPATCH_RigVMDispatch_ArrayGetAtIndex(in Array,in Index,out Element)"}.Element
    ).Transform
    ExecutePin=@RigVMFunction_Sequence.B
   ) {is=struct RigVMUnitNode}
   > RigVMVariableNode(Value=@n14.Item) {variable=Spline Point B}
   > n15 = &RigUnit_HierarchyAddBone(
    Name=Spline Point C
    Parent=(Type=Connector,Name="IKFK Virtual Bone B")
    Space=GlobalSpace
    Transform=&RigUnit_GetTransform(
     bInitial=False
     Item=RigVMDispatchNode(Index=4, Array=@n13.Result) {template="DISPATCH_RigVMDispatch_ArrayGetAtIndex(in Array,in Index,out Element)"}.Element
    ).Transform
   )
   > RigVMVariableNode(Value=@n15.Item) {variable=Spline Point C}
   > For_Each_1 = RigVMDispatchNode(Array=&RigUnit_ResolveArrayConnector(Connector=(Type=Connector,Name="Bones Chain"), SkipSocket=False).Result) {template="DISPATCH_RigVMDispatch_ArrayIterator(in Array,out Element,out Index,out Count,out Ratio)"}
   Int to Name = &RigVMFunction_MathIntToName(PaddedSize=0, Number=@For_Each_1.Index) {template="Int to Name::Execute(in Number,in PaddedSize,out Result)"}
   Concat = &RigVMFunction_NameConcat(A=Spline Bone 0, B=@"Int to Name".Result) {template="Concat::Execute(in A,in B,out Result)"}
   > n16 = &RigUnit_HierarchyAddBone(
    Space=GlobalSpace
    Name=@Concat.Result
    Transform=&RigUnit_GetTransform(bInitial=False, Item=@For_Each_1.Element).Transform
    Parent=RigVMVariableNode() {variable=Parent}.Value
   )
   > RigVMDispatchNode(Element=@n16.Item, Array=RigVMVariableNode() {variable=Spline Bones}.Value) {template="DISPATCH_RigVMDispatch_ArrayAdd(io Array,in Element,out Index)"}
  ]}
 RigVMModel Post Construction {nodes [
   n1 = /Script/ControlRig.RigUnit_PostPrepareForExecution() {is=struct RigVMUnitNode}
   RigVMFunction_Sequence = &RigVMFunction_Sequence(ExecuteContext=@n1.ExecutePin)
  ]}
]
hierarchy [
 Root {type=Connector, loc=<0 0 0>, rot=<0 0 0 1>}
 Root {type=Socket, loc=<0 0 0>, rot=<0 0 0 1>}
 Bones Chain {type=Connector, loc=<0 0 0>, rot=<0 0 0 1>}
 Parent {type=Connector, loc=<0 0 0>, rot=<0 0 0 1>}
 IKFK Virtual Bone A {type=Connector, loc=<0 0 0>, rot=<0 0 0 1>}
 IKFK Virtual Bone B {type=Connector, loc=<0 0 0>, rot=<0 0 0 1>}
 IKFK Virtual Bone C {type=Connector, loc=<0 0 0>, rot=<0 0 0 1>}
]
variables [
 Controls Scale {type=double, default=2.4}
 Draw Debug Lines {type=bool, default=False}
 Enable {type=bool, default=True}
 Mirror Behaviour {type=bool, default=False}
 Parent {type=FRigElementKey, default=(Type=Connector,Name="")}
 Spline Bones {type="TArray<FRigElementKey>", default=(())}
 Spline Control Offset {type=FTransform, default=(Rotation=(X=0,Y=0,Z=0,W=1),Translation=(X=0,Y=0,Z=0),Scale3D=(X=1,Y=1,Z=1))}
 Spline Control Settings {
  type=FRigUnit_HierarchyAddControlTransform_Settings
  default="(InitialSpace=LocalSpace,bUsePreferredRotationOrder=False,PreferredRotationOrder=YZX,Limits=(LimitTranslationX=(bMinimum=False,bMaximum=False),LimitTranslationY=(bMinimum=False,bMaximum=False),LimitTranslationZ=(bMinimum=False,bMaximum=False),LimitPitch=(b ...(truncated)"
 }
 Spline Controls {type="TArray<FRigElementKey>", default=(())}
 Spline Master Control {type=FRigElementKey, default=(Type=Bone,Name="")}
 Spline Point B {type=FRigElementKey, default=(Type=Bone,Name="")}
 Spline Point C {type=FRigElementKey, default=(Type=Bone,Name="")}
 Spline Switch Channel {type=FRigElementKey, default=(Type=Bone,Name="")}
 Sub Control Visibility Channel {type=FRigElementKey, default=(Type=Bone,Name="")}
 Volume Amount Channel {type=FRigElementKey, default=(Type=Bone,Name="")}
 Volume Falloff Channel {type=FRigElementKey, default=(Type=Bone,Name="")}
 Volume Position Channel {type=FRigElementKey, default=(Type=Bone,Name="")}
]
